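# Additive, unlit, two-sided mesh-particle material for the refracting energy shatter in LimeRock's TV suck-in effect.
# Emissive blends two offset, panning samples of stretched turbulence, desaturates them, adds a UV gradient and tints by particle color.
# Opacity combines vertex green and a vertical mask with view-angle, camera and depth fades.
# A particle dynamic parameter drives refraction; quality and feature-level switches simplify color, fades and refraction on low-end targets.
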Material /LimeRock/Assets/FX/TV_Suck_In/M_LimeRock_Energy_Shatter_Refract
graph {
 properties {
  bUsedWithMeshParticles=true
  bUsedWithNiagaraMeshParticles=true
  blendMode=BLEND_Additive
  pixelDepthOffsetMode=PDOM_Legacy
  refractionMethod=RM_IndexOfRefraction
  shadingModel=MSM_Unlit
  twoSided=true
  usedShadingModels=Unlit
 }
 nodes [
  n1 = ParticleColor()
  n2 = ComponentMask(None=TextureCoordinate()) {g=true}
  n3 = Multiply(
   A=Saturate(None=Power(Base=@n2) {constExponent=12})
   B=VectorParameter() {parameterName=BackgroundColor, defaultValue=<rgba 0.2031 0.2031 0.2031 0>}.RGB
  )
  n4 = Constant() {r=1}
  n5 = LinearInterpolate(Alpha=ComponentMask(None=TextureCoordinate() {coordinateIndex=1}) {g=true}) {constA=0.7}
  n6 = Add(B=Multiply(A=ScalarParameter() {defaultValue=0.07, parameterName=ChromaOffset}, B=@n5) {constA=1}) {constA=1}
  n7 = VectorParameter() {parameterName=Tile, defaultValue=<rgba 1.5 0.8 0 0>}
  n8 = Multiply(A=AppendVector(A=@n7.R, B=Multiply(A=@n7.G, B=@n5)), B=TextureCoordinate())
  n9 = Add(A=Multiply(A=AppendVector(A=@n4, B=@n6), B=@n8), B=Multiply(A=ScalarParameter() {parameterName=ChromaOffsetAdd}, B=Constant2Vector() {g=1}))
  n10 = Multiply(A=ScalarParameter() {defaultValue=0.1, parameterName=PanSpeed}, B=Time())
  n11 = Multiply(
   A=VectorParameter() {parameterName=Color_Low, defaultValue=<rgba 1 1 1 1>}.RGB
   B=TextureSample(UVs=Panner(Coordinate=@n9, Time=@n10) {speedX=0.04, speedY=0.12}) {texture=/Game/Effects/Fort_Effects/Textures/Plasma/T_Turbulence_Stretched}.G
  )
  n12 = Multiply(
   A=Constant3Vector() {constant=<rgba 0.0312 0.0308 0.0308 0>}
   B=TextureSample(UVs=Panner(Coordinate=@n8, Time=@n10) {speedX=0.04, speedY=0.12}) {texture=&T_Turbulence_Stretched}.G
  )
  n13 = QualitySwitch(Default=Desaturation(None=LinearInterpolate(A=@n11, B=@n12), Fraction=ScalarParameter() {parameterName=Saturation}), Low=@n11)
  MaterialFunctionCall(Emissive=Multiply(A=@n1.RGB, B=Add(A=@n3, B=@n13))) {materialFunction=&MF_FX_EmissiveAdjustments}
  n14 = ComponentMask(None=TextureCoordinate()) {g=true}
  n15 = Constant() {r=0.5}
  n16 = Multiply(
   A=VertexColor().G
   B=SphereMask(A=@n14, B=@n15, Radius=@n15, Hardness=ScalarParameter() {defaultValue=0.25, parameterName=VerticalMaskHardness}) {hardnessPercent=64}
  )
  n17 = ScalarParameter() {defaultValue=1, parameterName=AngleFade}
  n18 = ScalarParameter() {parameterName=Camera Fade Distance}
  n19 = Multiply(
   A=Multiply(
    A=MaterialFunctionCall("FadeSensitivity (optional)"=@n17) {materialFunction=&FadeBasedOnViewAngle}.Result
    B=MaterialFunctionCall(Distance=@n18) {materialFunction=&Near_Camera_Fade}.Result
   )
   B=DepthFade(FadeDistance=ScalarParameter() {defaultValue=500, parameterName=FadeDistance})
  )
  n20 = Constant() {r=1}
  Multiply(A=@n1.A, B=Multiply(A=Power(Base=@n16) {constExponent=2.5}, B=FeatureLevelSwitch(Default=@n19, ES2=@n20, ES3_1=@n20, SM4_REMOVED=@n19)))
  n21 = DynamicParameter() {paramNames [refraction,Param2,Param3,Param4]}
  Multiply(A=ScalarParameter() {defaultValue=1.05, parameterName=Refraction}, B=@n21.refraction)
  n22 = LinearInterpolate(B=@n21.refraction, Alpha=Multiply(A=@n19, B=@n14)) {constA=1}
  n23 = Constant() {r=1}
  QualitySwitch(Default=FeatureLevelSwitch(Default=@n22, ES2=@n23, ES3_1=@n23, SM4_REMOVED=@n22), Low=@n23)
 ]
}
